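# Control Rig for the female medium Valor_Head_01 cosmetic head, adding a head-pose corrective signal and re-anchoring one hair chain.
# A spherical pose reader on the head bone, with debug drawing on, writes its output to the head_SPR curve.
# Full-weight parent constraints then make dyn_hair_05 and its _a and _b children follow spine_05, although the chain is parented under the head.
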ControlRigBlueprint /BRCosmetics/Characters/Player/Female/Medium/Heads/F_MED_IND_Valor_Head_01/Meshes/F_MED_IND_Valor_Head_01_CtrlRig
graphs [
 RigVMFunctionLibrary {links [], nodes []}
 RigVMModel {nodes [
   RigUnit_BeginExecution = &RigUnit_BeginExecution()
   n1 = &RigUnit_SphericalPoseReader(
    ActiveRegionScaleFactors=(PositiveWidth=1,NegativeWidth=1,PositiveHeight=1,NegativeHeight=1)
    ActiveRegionSize=0.1
    Debug=(bDrawDebug=True,bDraw2D=False,bDrawLocalAxes=False,DebugScale=25,DebugSegments=20,DebugThickness=0.25)
    DriverAxis=(X=0,Y=1,Z=0)
    DriverItem=(Type=Bone,Name="head")
    FalloffRegionScaleFactors=(PositiveWidth=1,NegativeWidth=1,PositiveHeight=1,NegativeHeight=1)
    FalloffSize=0.3
    FlipHeightScaling=False
    FlipWidthScaling=False
    OptionalParentItem=(Type=Bone,Name="None")
    RotationOffset=(X=90,Y=0,Z=-35)
    ExecutePin=@RigUnit_BeginExecution.ExecutePin
   )
   > /Script/ControlRig.RigUnit_SetCurveValue(Curve=head_SPR, Value=@n1.OutputParam) {is=struct RigVMUnitNode}
   > &RigUnit_ParentConstraint(
    AdvancedSettings=(InterpolationType=Average,RotationOrderForFilter=XZY)
    Child=(Type=Bone,Name="dyn_hair_05")
    Parents=((Item=(Type=Bone,Name="spine_05"),Weight=1))
    Weight=1
   )
   > &RigUnit_ParentConstraint(
    AdvancedSettings=(InterpolationType=Average,RotationOrderForFilter=XZY)
    Child=(Type=Bone,Name="dyn_hair_05_a")
    Parents=((Item=(Type=Bone,Name="spine_05"),Weight=1))
    Weight=1
   )
   > &RigUnit_ParentConstraint(
    AdvancedSettings=(InterpolationType=Average,RotationOrderForFilter=XZY)
    Child=(Type=Bone,Name="dyn_hair_05_b")
    Parents=((Item=(Type=Bone,Name="spine_05"),Weight=1))
    Weight=1
   )
  ]}
]
hierarchy [
 root {loc=<0 0 0>, rot=<0 0 0 1>}
 attach {parent=root, loc=<0 0 0>, rot=<0 0 0 1>}
 pelvis {parent=root, loc=<0 -0.5301 98.0571>, rot=<0.0127 0.707 -0.0127 -0.707>}
 spine_01 {parent=pelvis, loc=<6.1208 -0.9328 0>, rot=<0 0 0.034 -0.9994>}
 spine_02 {parent=spine_01, loc=<6.5996 -0.2095 0>, rot=<0 0 0.0238 -0.9997>}
 spine_03 {parent=spine_02, loc=<6.4914 -0.1237 0>, rot=<0 0 0.0149 -0.9999>}
 spine_04 {parent=spine_03, loc=<6.9598 0.0794 0>, rot=<0 0 -0.0132 -0.9999>}
 spine_05 {parent=spine_04, loc=<6.9987 0 0>, rot=<0 0 -0.0015 -1>}
 clavicle_l {parent=spine_05, loc=<5.4831 -0.1053 -3.2946>, rot=<0.0465 -0.7289 -0.009 -0.6829>}
 upperarm_l {parent=clavicle_l, loc=<9.8658 -1.2136 -0.4105>, rot=<0.0208 -0.486 -0.0242 -0.8734>}
 lowerarm_l {parent=upperarm_l, loc=<26.465 -0.8345 0.259>, rot=<0.0036 0.0161 -0.2298 -0.9731>}
 hand_l {parent=lowerarm_l, loc=<22.7355 -0.3203 -0.6105>, rot=<-0.6784 -0.0223 0.0735 -0.7307>}
 index_metacarpal_l {parent=hand_l, loc=<1.6535 0.0183 -2.96>, rot=<0.0158 -0.0124 -0.0173 -0.9996>}
 index_01_l {parent=index_metacarpal_l, loc=<8.3008 0.0357 -0.0524>, rot=<-0.0044 0.0485 0.0924 -0.9945>}
 index_02_l {parent=index_01_l, loc=<3.3587 -0.3004 0.0452>, rot=<-0.002 0.0027 0.3225 -0.9466>}
 index_03_l {parent=index_02_l, loc=<2.5578 -0.123 -0.0252>, rot=<-0.002 0.0067 0.2558 -0.9667>}
 middle_metacarpal_l {parent=hand_l, loc=<2.1856 -0.1725 -0.5984>, rot=<0.006 0.0168 -0.0329 -0.9993>}
 middle_01_l {parent=middle_metacarpal_l, loc=<7.3275 -0.2447 0.0129>, rot=<-0.0208 0.0042 0.0858 -0.9961>}
 middle_02_l {parent=middle_01_l, loc=<4.0123 0.1401 0.1175>, rot=<-0.0218 0.0258 0.3512 -0.9357>}
 middle_03_l {parent=middle_02_l, loc=<2.6077 0.0534 0.1373>, rot=<0.0082 0.0188 0.1708 -0.9851>}
 pinky_metacarpal_l {parent=hand_l, loc=<1.7616 -0.9753 3.3814>, rot=<-0.0397 0.0814 -0.069 -0.9935>}
 pinky_01_l {parent=pinky_metacarpal_l, loc=<7.1709 -0.4831 -0.1094>, rot=<-0.0654 -0.0035 0.2242 -0.9723>}
 pinky_02_l {parent=pinky_01_l, loc=<2.8274 0.0893 -0.1043>, rot=<0.0037 0.0011 0.2854 -0.9584>}
 pinky_03_l {parent=pinky_02_l, loc=<2.2046 0.0307 -0.0078>, rot=<-0.0002 0.0098 0.1445 -0.9895>}
 ring_metacarpal_l {parent=hand_l, loc=<2.2074 -0.3625 1.3638>, rot=<-0.0284 0.0722 -0.0458 -0.9959>}
 ring_01_l {parent=ring_metacarpal_l, loc=<7.0295 -0.6197 -0.1446>, rot=<-0.0489 -0.0252 0.1892 -0.9804>}
 ring_02_l {parent=ring_01_l, loc=<3.6174 0.4647 -0.1027>, rot=<0 0 0.2806 -0.9598>}
 ring_03_l {parent=ring_02_l, loc=<2.4152 0.114 0.0199>, rot=<0 0 0.1189 -0.9929>}
 thumb_01_l {parent=hand_l, loc=<1.3118 -0.3191 -2.459>, rot=<0.6546 -0.4042 0.0664 -0.6354>}
 thumb_02_l {parent=thumb_01_l, loc=<5.9336 -0.0448 0.0401>, rot=<-0.0037 0.0213 0.298 -0.9543>}
 thumb_03_l {parent=thumb_02_l, loc=<2.2223 0.3682 -0.0907>, rot=<0.0118 -0.013 0.0066 -0.9998>}
 weapon_l {parent=hand_l, loc=<-0.1974 -3.4111 0.3811>, rot=<0 0 0 1>}
 lowerarm_twist_01_l {parent=lowerarm_l, loc=<18.316 0 0>, rot=<0 0 0 1>}
 lowerarm_twist_02_l {parent=lowerarm_l, loc=<10.6707 0 0>, rot=<0 0 0 1>}
 upperarm_twist_01_l {parent=upperarm_l, loc=<7.2135 -0.1134 0>, rot=<0 0 0 1>}
 upperarm_twist_02_l {parent=upperarm_l, loc=<19.7998 -0.1134 0>, rot=<0 0 0 1>}
 clavicle_r {parent=spine_05, loc=<5.4831 -0.1053 3.2946>, rot=<0.7289 0.0465 -0.6829 0.009>}
 upperarm_r {parent=clavicle_r, loc=<-9.8658 1.2136 0.4104>, rot=<0.0208 -0.486 -0.0242 -0.8734>}
 lowerarm_r {parent=upperarm_r, loc=<-26.4651 0.8345 -0.259>, rot=<0.0036 0.0161 -0.2298 -0.9731>}
 hand_r {parent=lowerarm_r, loc=<-22.7356 0.3203 0.6106>, rot=<-0.6784 -0.0223 0.0735 -0.7307>}
 index_metacarpal_r {parent=hand_r, loc=<-1.6534 -0.0181 2.9601>, rot=<0.0158 -0.0124 -0.0173 -0.9996>}
 index_01_r {parent=index_metacarpal_r, loc=<-8.3008 -0.0357 0.0524>, rot=<-0.0044 0.0485 0.0924 -0.9945>}
 index_02_r {parent=index_01_r, loc=<-3.3589 0.3003 -0.0453>, rot=<-0.002 0.0027 0.3225 -0.9466>}
 index_03_r {parent=index_02_r, loc=<-2.5578 0.123 0.0252>, rot=<-0.002 0.0067 0.2558 -0.9667>}
 middle_metacarpal_r {parent=hand_r, loc=<-2.1856 0.1726 0.5984>, rot=<0.006 0.0168 -0.0329 -0.9993>}
 middle_01_r {parent=middle_metacarpal_r, loc=<-7.3273 0.2447 -0.0129>, rot=<-0.0208 0.0042 0.0858 -0.9961>}
 middle_02_r {parent=middle_01_r, loc=<-4.0124 -0.14 -0.1175>, rot=<-0.0218 0.0258 0.3512 -0.9357>}
 middle_03_r {parent=middle_02_r, loc=<-2.6077 -0.0533 -0.1373>, rot=<0.0082 0.0188 0.1708 -0.9851>}
 pinky_metacarpal_r {parent=hand_r, loc=<-1.7617 0.9753 -3.3814>, rot=<-0.0397 0.0814 -0.069 -0.9935>}
 pinky_01_r {parent=pinky_metacarpal_r, loc=<-7.1707 0.4832 0.1094>, rot=<-0.0654 -0.0035 0.2242 -0.9723>}
 pinky_02_r {parent=pinky_01_r, loc=<-2.8274 -0.0892 0.1043>, rot=<0.0037 0.0011 0.2854 -0.9584>}
 pinky_03_r {parent=pinky_02_r, loc=<-2.2045 -0.031 0.0078>, rot=<-0.0002 0.0098 0.1445 -0.9895>}
 ring_metacarpal_r {parent=hand_r, loc=<-2.2073 0.3627 -1.3638>, rot=<-0.0284 0.0722 -0.0458 -0.9959>}
 ring_01_r {parent=ring_metacarpal_r, loc=<-7.0297 0.6197 0.1446>, rot=<-0.0489 -0.0252 0.1892 -0.9804>}
 ring_02_r {parent=ring_01_r, loc=<-3.6171 -0.4648 0.1027>, rot=<0 0 0.2806 -0.9598>}
 ring_03_r {parent=ring_02_r, loc=<-2.4153 -0.1141 -0.0199>, rot=<0 0 0.1189 -0.9929>}
 thumb_01_r {parent=hand_r, loc=<-1.3119 0.3192 2.459>, rot=<0.6546 -0.4042 0.0664 -0.6354>}
 thumb_02_r {parent=thumb_01_r, loc=<-5.9335 0.0447 -0.0401>, rot=<-0.0037 0.0213 0.298 -0.9543>}
 thumb_03_r {parent=thumb_02_r, loc=<-2.2223 -0.3682 0.0907>, rot=<0.0118 -0.013 0.0066 -0.9998>}
 weapon_r {parent=hand_r, loc=<0.1975 3.4112 -0.3811>, rot=<0 0 0 -1>}
 lowerarm_twist_01_r {parent=lowerarm_r, loc=<-18.3161 0 0>, rot=<0 0 0 1>}
 lowerarm_twist_02_r {parent=lowerarm_r, loc=<-10.6707 0 0>, rot=<0 0 0 1>}
 upperarm_twist_01_r {parent=upperarm_r, loc=<-7.2135 0.1134 0>, rot=<0 0 0 1>}
 upperarm_twist_02_r {parent=upperarm_r, loc=<-19.7998 0.1134 0>, rot=<0 0 0 1>}
 neck_01 {parent=spine_05, loc=<11.3138 0.1529 0>, rot=<0 0 -0.2003 -0.9797>}
 neck_02 {parent=neck_01, loc=<4.5627 0 0>, rot=<0 0 0.0504 -0.9987>}
 head {parent=neck_02, loc=<5.3914 0 0>, rot=<0 0 0.115 -0.9934>}
 faceAttach {parent=head, loc=<-4.171 -0.6579 0.0207>, rot=<0.0082 -0.7071 -0.0082 -0.7071>, scale=<0.89 0.89 0.89>}
 C_brow_mid {parent=faceAttach, loc=<0.0008 10.2084 9.0994>, rot=<0 -0.032 0.0063 -0.9995>}
 L_brow_mid {parent=faceAttach, loc=<3.6796 9.7338 10.6839>, rot=<0 -0.0036 0.2553 -0.9669>}
 L_brow_outer {parent=faceAttach, loc=<6.1837 7.439 10.6464>, rot=<0.1601 0.1977 0.3899 -0.885>}
 L_cheek_inner {parent=faceAttach, loc=<4.6333 8.8024 4.1023>, rot=<-0.0217 -0.1457 0.3327 -0.9315>}
 C_lip_upper_mid {parent=faceAttach, loc=<0.0388 10.2772 1.3932>, rot=<0.2225 0.0019 -0.0007 -0.9749>}
 L_lip_upper_outer {parent=faceAttach, loc=<1.8577 9.4904 1.3497>, rot=<0.0484 -0.0377 0.2493 -0.9665>}
 L_lip_corner {parent=faceAttach, loc=<2.8126 8.7014 1.2955>, rot=<0.0023 0.0087 -0.0145 -0.9999>}
 C_nose_bridge {parent=faceAttach, loc=<0.0201 11.9613 5.5054>, rot=<-0.2218 -0.0046 -0.0013 -0.9751>}
 R_brow_mid {parent=faceAttach, loc=<-3.6322 9.7354 10.6891>, rot=<-0.0001 0.0043 -0.2553 -0.9669>}
 R_brow_outer {parent=faceAttach, loc=<-6.1377 7.436 10.6221>, rot=<0.0921 -0.112 -0.4162 -0.8976>}
 R_cheek_inner {parent=faceAttach, loc=<-4.5746 8.8088 4.0924>, rot=<-0.0217 0.1457 -0.3327 -0.9315>}
 R_lip_corner {parent=faceAttach, loc=<-2.7769 8.7197 1.3267>, rot=<0.0027 -0.0082 0.0146 -0.9999>}
 C_jaw {parent=faceAttach, loc=<0.0211 2.9336 1.4734>, rot=<0 0 0 1>}
 L_lip_lower_outer {parent=C_jaw, loc=<1.8024 6.2805 -0.1414>, rot=<0.015 0.0555 0.314 -0.9477>}
 C_lip_lower_mid {parent=C_jaw, loc=<0.0063 7.0726 -0.243>, rot=<0 -0.0135 -0.0029 -0.9999>}
 R_lip_lower_outer {parent=C_jaw, loc=<-1.8425 6.414 -0.2145>, rot=<0.0051 -0.0015 -0.2727 -0.9621>}
 teeth_lower {parent=C_jaw, loc=<-0.1572 4.2115 -2.1085>, rot=<0 0 0 1>}
 tongue {parent=C_jaw, loc=<-0.0068 1.6174 -0.8453>, rot=<-0.1568 0 0 -0.9876>}
 R_eye {parent=faceAttach, loc=<-3.6268 7.3931 8.4073>, rot=<0 0 0.9998 -0.0187>}
 R_eye_lid_upper_mid {parent=R_eye, loc=<-0.0202 -1.9262 1.0798>, rot=<0.0523 -0.0009 0.0167 -0.9985>}
 R_eye_lid_lower_mid {parent=R_eye, loc=<-0.0556 -1.4492 -1.3852>, rot=<0.0329 -0.0005 0.0167 -0.9993>}
 L_eye {parent=faceAttach, loc=<3.6783 7.391 8.4075>, rot=<0 0 0.9998 0.0187>}
 L_eye_lid_upper_mid {parent=L_eye, loc=<-0.2075 -1.9319 1.1122>, rot=<0.0523 0 0 -0.9986>}
 L_eye_lid_lower_mid {parent=L_eye, loc=<0.0746 -1.4324 -1.4294>, rot=<0.0329 0 0 -0.9995>}
 teeth_upper {parent=faceAttach, loc=<0.0025 8.0008 2.382>, rot=<0 0 0 1>}
 R_lip_upper_outer {parent=faceAttach, loc=<-1.8235 9.5252 1.4027>, rot=<0.05 0.0329 -0.2475 -0.967>}
 dyn_hair_04 {parent=head, loc=<7.8987 6.2939 -5.1285>, rot=<0.0029 -0.7071 -0.0029 -0.7071>}
 dyn_hair_07 {parent=head, loc=<7.9303 0.6072 -10.0359>, rot=<0.0029 -0.7071 -0.0029 -0.7071>}
 dyn_hair_07_a {parent=dyn_hair_07, loc=<1.0156 -3.6084 -3.4193>, rot=<0 0 0 1>}
 dyn_hair_07_b {parent=dyn_hair_07_a, loc=<3.708 -7.9342 -3.6194>, rot=<0 0 0 1>}
 dyn_hair_06 {parent=head, loc=<13.1539 -10.0436 0>, rot=<0.0029 -0.7071 -0.0029 -0.7071>}
 dyn_hair_06_a {parent=dyn_hair_06, loc=<0 -5.3624 4.8009>, rot=<0 0 0 1>}
 dyn_hair_06_b {parent=dyn_hair_06_a, loc=<0 -2.9962 -4.2421>, rot=<0 0 0 1>}
 dyn_hair_06_c {parent=dyn_hair_06_b, loc=<0 2.1655 -6.1703>, rot=<0 0 0 1>}
 dyn_hair_05 {parent=head, loc=<2.1341 -6.908 0>, rot=<0.0029 -0.7071 -0.0029 -0.7071>}
 dyn_hair_05_a {parent=dyn_hair_05, loc=<0 -5.568 -6.0346>, rot=<0 0 0 1>}
 dyn_hair_05_b {parent=dyn_hair_05_a, loc=<0 -1.0887 -6.3457>, rot=<0 0 0 1>}
 dyn_hair_03 {parent=head, loc=<-2.1622 -4.4773 -13.5341>, rot=<0.2678 -0.8165 0.3917 -0.3288>}
 dyn_earing_l {parent=head, loc=<0.5802 -0.2012 -7.078>, rot=<0.0029 -0.7071 -0.0029 -0.7071>}
 dyn_earing_r {parent=head, loc=<0.5805 -0.2012 7.078>, rot=<0.0029 -0.7071 -0.0029 -0.7071>}
 dyn_hair_02 {parent=head, loc=<-2.1623 -4.4773 13.5341>, rot=<0.2678 -0.8165 0.3917 -0.3288>}
 dyn_hair_01 {parent=head, loc=<7.899 6.2939 5.1285>, rot=<0.0029 -0.7071 -0.0029 -0.7071>}
 dyn_hair_08 {parent=head, loc=<7.93 0.6072 10.0359>, rot=<0.0029 -0.7071 -0.0029 -0.7071>}
 dyn_hair_08_a {parent=dyn_hair_08, loc=<-1.0155 -3.6084 -3.419>, rot=<0 0 0 1>}
 dyn_hair_08_c {parent=dyn_hair_08_a, loc=<-4.1176 -8.381 -4.362>, rot=<0 0 0 1>}
 thigh_l {parent=pelvis, loc=<-1.0855 -0.2109 -8.7053>, rot=<0.0009 0.0355 0.0155 -0.9993>}
 calf_l {parent=thigh_l, loc=<-42.5284 -0.0019 -0.7297>, rot=<0.0006 -0.0161 -0.0391 -0.9991>}
 calf_twist_01_l {parent=calf_l, loc=<-35.8483 -1.9197 -0.5794>, rot=<0.0237 0.0095 -0.0035 -0.9997>}
 calf_twist_02_l {parent=calf_l, loc=<-19.1546 -1.7943 -0.2638>, rot=<0.0237 0.0095 -0.0035 -0.9997>}
 foot_l {parent=calf_l, loc=<-46.7402 -2.5014 -0.592>, rot=<0.0427 -0.0238 0.0353 -0.9982>}
 ball_l {parent=foot_l, loc=<-3.5658 10.8066 -0.1544>, rot=<0.0001 -0.0001 -0.7063 -0.7079>}
 thigh_twist_01_l {parent=thigh_l, loc=<-27.952 0 0>, rot=<0.0013 0 0 -1>}
 thigh_r {parent=pelvis, loc=<-1.0853 -0.2109 8.7053>, rot=<0.0355 -0.0009 0.9993 0.0155>}
 calf_r {parent=thigh_r, loc=<42.5286 0.0019 0.7297>, rot=<0.0006 -0.0161 -0.0391 -0.9991>}
 calf_twist_01_r {parent=calf_r, loc=<35.8485 1.9197 0.5794>, rot=<0.0237 0.0095 -0.0035 -0.9997>}
 calf_twist_02_r {parent=calf_r, loc=<19.1546 1.7943 0.2638>, rot=<0.0237 0.0095 -0.0035 -0.9997>}
 foot_r {parent=calf_r, loc=<46.7403 2.5014 0.592>, rot=<0.0427 -0.0238 0.0353 -0.9982>}
 ball_r {parent=foot_r, loc=<3.5658 -10.8066 0.1544>, rot=<0.0001 -0.0001 -0.7063 -0.7079>}
 thigh_twist_01_r {parent=thigh_r, loc=<27.952 0 0>, rot=<0.0013 0 0 -1>}
 ik_foot_root {parent=root, loc=<0 0 0>, rot=<0 0 0 1>}
 ik_foot_l {parent=ik_foot_root, loc=<14.8262 -7.3545 8.296>, rot=<0.0357 0.7045 0.0284 -0.7082>}
 ik_foot_r {parent=ik_foot_root, loc=<-14.8262 -7.3546 8.2958>, rot=<0.7082 0.0284 -0.7045 0.0357>}
 ik_hand_root {parent=root, loc=<0 0 0>, rot=<0 0 0 1>}
 ik_hand_gun {parent=ik_hand_root, loc=<-35.2227 3.1613 94.1496>, rot=<0.5451 0.2436 0.5214 -0.6097>}
 ik_hand_l {parent=ik_hand_gun, loc=<23.7864 63.4948 19.1095>, rot=<-0.3377 -0.9013 -0.2713 0>}
 ik_hand_r {parent=ik_hand_gun, loc=<0 0 0>, rot=<0 0 0 1>}
 look_left_pose {type=Curve, loc=<0 0 0>, rot=<0 0 0 1>}
 look_right_pose {type=Curve, loc=<0 0 0>, rot=<0 0 0 1>}
 look_up_pose {type=Curve, loc=<0 0 0>, rot=<0 0 0 1>}
 look_down_pose {type=Curve, loc=<0 0 0>, rot=<0 0 0 1>}
 L_blink_pose {type=Curve, loc=<0 0 0>, rot=<0 0 0 1>}
 R_blink_pose {type=Curve, loc=<0 0 0>, rot=<0 0 0 1>}
 L_wide_pose {type=Curve, loc=<0 0 0>, rot=<0 0 0 1>}
 R_wide_pose {type=Curve, loc=<0 0 0>, rot=<0 0 0 1>}
 R_squint_inner_pose {type=Curve, loc=<0 0 0>, rot=<0 0 0 1>}
 R_squeeze_pose {type=Curve, loc=<0 0 0>, rot=<0 0 0 1>}
 L_squint_inner_pose {type=Curve, loc=<0 0 0>, rot=<0 0 0 1>}
 L_squeeze_pose {type=Curve, loc=<0 0 0>, rot=<0 0 0 1>}
 L_brow_up_pose {type=Curve, loc=<0 0 0>, rot=<0 0 0 1>}
 L_brow_down_pose {type=Curve, loc=<0 0 0>, rot=<0 0 0 1>}
 R_brow_up_pose {type=Curve, loc=<0 0 0>, rot=<0 0 0 1>}
 R_brow_down_pose {type=Curve, loc=<0 0 0>, rot=<0 0 0 1>}
 C_glabella_up_pose {type=Curve, loc=<0 0 0>, rot=<0 0 0 1>}
 C_glabella_down_pose {type=Curve, loc=<0 0 0>, rot=<0 0 0 1>}
 L_brow_squeeze_in_pose {type=Curve, loc=<0 0 0>, rot=<0 0 0 1>}
 R_brow_squeeze_in_pose {type=Curve, loc=<0 0 0>, rot=<0 0 0 1>}
 L_upper_lip_raiser_pose {type=Curve, loc=<0 0 0>, rot=<0 0 0 1>}
 R_upper_lip_raiser_pose {type=Curve, loc=<0 0 0>, rot=<0 0 0 1>}
 L_upper_lip_lower_pose {type=Curve, loc=<0 0 0>, rot=<0 0 0 1>}
 R_upper_lip_lower_pose {type=Curve, loc=<0 0 0>, rot=<0 0 0 1>}
 L_lower_lip_up_pose {type=Curve, loc=<0 0 0>, rot=<0 0 0 1>}
 R_lower_lip_up_pose {type=Curve, loc=<0 0 0>, rot=<0 0 0 1>}
 L_lower_lip_down_pose {type=Curve, loc=<0 0 0>, rot=<0 0 0 1>}
 R_lower_lip_down_pose {type=Curve, loc=<0 0 0>, rot=<0 0 0 1>}
 L_lip_corner_narrow_pose {type=Curve, loc=<0 0 0>, rot=<0 0 0 1>}
 L_lip_corner_wide_pose {type=Curve, loc=<0 0 0>, rot=<0 0 0 1>}
 R_lip_corner_narrow_pose {type=Curve, loc=<0 0 0>, rot=<0 0 0 1>}
 R_lip_corner_wide_pose {type=Curve, loc=<0 0 0>, rot=<0 0 0 1>}
 L_smile_pose {type=Curve, loc=<0 0 0>, rot=<0 0 0 1>}
 R_smile_pose {type=Curve, loc=<0 0 0>, rot=<0 0 0 1>}
 L_frown_pose {type=Curve, loc=<0 0 0>, rot=<0 0 0 1>}
 R_frown_pose {type=Curve, loc=<0 0 0>, rot=<0 0 0 1>}
 phoneme_oo_pose {type=Curve, loc=<0 0 0>, rot=<0 0 0 1>}
 phoneme_ch_pose {type=Curve, loc=<0 0 0>, rot=<0 0 0 1>}
 phoneme_mbp_pose {type=Curve, loc=<0 0 0>, rot=<0 0 0 1>}
 phoneme_fv_pose {type=Curve, loc=<0 0 0>, rot=<0 0 0 1>}
 mouth_up_pose {type=Curve, loc=<0 0 0>, rot=<0 0 0 1>}
 mouth_down_pose {type=Curve, loc=<0 0 0>, rot=<0 0 0 1>}
 C_nose_wrinkler_pose {type=Curve, loc=<0 0 0>, rot=<0 0 0 1>}
 L_cheek_up_pose {type=Curve, loc=<0 0 0>, rot=<0 0 0 1>}
 R_cheek_up_pose {type=Curve, loc=<0 0 0>, rot=<0 0 0 1>}
 jaw_open_pose {type=Curve, loc=<0 0 0>, rot=<0 0 0 1>}
 jaw_right_pose {type=Curve, loc=<0 0 0>, rot=<0 0 0 1>}
 jaw_left_pose {type=Curve, loc=<0 0 0>, rot=<0 0 0 1>}
 eye_convergence_pose {type=Curve, loc=<0 0 0>, rot=<0 0 0 1>}
 L_upperLidDown_pose {type=Curve, loc=<0 0 0>, rot=<0 0 0 1>}
 R_upperLidDown_pose {type=Curve, loc=<0 0 0>, rot=<0 0 0 1>}
 jaw_up_pose {type=Curve, loc=<0 0 0>, rot=<0 0 0 1>}
 jaw_forward_pose {type=Curve, loc=<0 0 0>, rot=<0 0 0 1>}
 jaw_back_pose {type=Curve, loc=<0 0 0>, rot=<0 0 0 1>}
 tongue_forward_pose {type=Curve, loc=<0 0 0>, rot=<0 0 0 1>}
 tongue_back_pose {type=Curve, loc=<0 0 0>, rot=<0 0 0 1>}
 tongue_up_pose {type=Curve, loc=<0 0 0>, rot=<0 0 0 1>}
 tongue_down_pose {type=Curve, loc=<0 0 0>, rot=<0 0 0 1>}
 base_pose {type=Curve, loc=<0 0 0>, rot=<0 0 0 1>}
 Hair_Scarf {type=Curve, loc=<0 0 0>, rot=<0 0 0 1>}
 Hair_Hat {type=Curve, loc=<0 0 0>, rot=<0 0 0 1>}
 Hair_Helmet {type=Curve, loc=<0 0 0>, rot=<0 0 0 1>}
 Hair_Glasses {type=Curve, loc=<0 0 0>, rot=<0 0 0 1>}
 Hair_Cap {type=Curve, loc=<0 0 0>, rot=<0 0 0 1>}
 Hair_Headsets {type=Curve, loc=<0 0 0>, rot=<0 0 0 1>}
 Hair_Hood {type=Curve, loc=<0 0 0>, rot=<0 0 0 1>}
 Face_Mask {type=Curve, loc=<0 0 0>, rot=<0 0 0 1>}
 HairStrands_Hat {type=Curve, loc=<0 0 0>, rot=<0 0 0 1>}
 Hat_Hair {type=Curve, loc=<0 0 0>, rot=<0 0 0 1>}
 Helmet_Hair {type=Curve, loc=<0 0 0>, rot=<0 0 0 1>}
 Hair_Muff {type=Curve, loc=<0 0 0>, rot=<0 0 0 1>}
 Hair_SpaceSuit {type=Curve, loc=<0 0 0>, rot=<0 0 0 1>}
 Hair_Hide {type=Curve, loc=<0 0 0>, rot=<0 0 0 1>}
 Head_SpaceSuit {type=Curve, loc=<0 0 0>, rot=<0 0 0 1>}
 Hair_Null {type=Curve, loc=<0 0 0>, rot=<0 0 0 1>}
 DisableFaceOverride {type=Curve, loc=<0 0 0>, rot=<0 0 0 1>}
 LiveLink {type=Curve, loc=<0 0 0>, rot=<0 0 0 1>}
 Hat_MESH {type=Curve, loc=<0 0 0>, rot=<0 0 0 1>}
 Null_MESH {type=Curve, loc=<0 0 0>, rot=<0 0 0 1>}
 Pose_0 {type=Curve, loc=<0 0 0>, rot=<0 0 0 1>}
 Hair_Mask {type=Curve, loc=<0 0 0>, rot=<0 0 0 1>}
 eye_convergence {type=Curve, loc=<0 0 0>, rot=<0 0 0 1>}
 headTransform {type=Curve, loc=<0 0 0>, rot=<0 0 0 1>}
 Pose_1 {type=Curve, loc=<0 0 0>, rot=<0 0 0 1>}
 Pose_2 {type=Curve, loc=<0 0 0>, rot=<0 0 0 1>}
 Pose_3 {type=Curve, loc=<0 0 0>, rot=<0 0 0 1>}
 Pose_4 {type=Curve, loc=<0 0 0>, rot=<0 0 0 1>}
 Pose_5 {type=Curve, loc=<0 0 0>, rot=<0 0 0 1>}
 Pose_6 {type=Curve, loc=<0 0 0>, rot=<0 0 0 1>}
 Pose_7 {type=Curve, loc=<0 0 0>, rot=<0 0 0 1>}
 Pose_8 {type=Curve, loc=<0 0 0>, rot=<0 0 0 1>}
 Pose_9 {type=Curve, loc=<0 0 0>, rot=<0 0 0 1>}
 Pose_10 {type=Curve, loc=<0 0 0>, rot=<0 0 0 1>}
 Pose_11 {type=Curve, loc=<0 0 0>, rot=<0 0 0 1>}
 Pose_12 {type=Curve, loc=<0 0 0>, rot=<0 0 0 1>}
 Pose_13 {type=Curve, loc=<0 0 0>, rot=<0 0 0 1>}
 Pose_14 {type=Curve, loc=<0 0 0>, rot=<0 0 0 1>}
 Pose_15 {type=Curve, loc=<0 0 0>, rot=<0 0 0 1>}
 Pose_16 {type=Curve, loc=<0 0 0>, rot=<0 0 0 1>}
 Pose_17 {type=Curve, loc=<0 0 0>, rot=<0 0 0 1>}
 Pose_18 {type=Curve, loc=<0 0 0>, rot=<0 0 0 1>}
 Pose_19 {type=Curve, loc=<0 0 0>, rot=<0 0 0 1>}
 Pose_20 {type=Curve, loc=<0 0 0>, rot=<0 0 0 1>}
 Pose_21 {type=Curve, loc=<0 0 0>, rot=<0 0 0 1>}
 Pose_22 {type=Curve, loc=<0 0 0>, rot=<0 0 0 1>}
 Pose_23 {type=Curve, loc=<0 0 0>, rot=<0 0 0 1>}
 Pose_24 {type=Curve, loc=<0 0 0>, rot=<0 0 0 1>}
 Pose_25 {type=Curve, loc=<0 0 0>, rot=<0 0 0 1>}
 Pose_26 {type=Curve, loc=<0 0 0>, rot=<0 0 0 1>}
 Pose_27 {type=Curve, loc=<0 0 0>, rot=<0 0 0 1>}
 Pose_28 {type=Curve, loc=<0 0 0>, rot=<0 0 0 1>}
 Pose_29 {type=Curve, loc=<0 0 0>, rot=<0 0 0 1>}
 Pose_30 {type=Curve, loc=<0 0 0>, rot=<0 0 0 1>}
 Pose_31 {type=Curve, loc=<0 0 0>, rot=<0 0 0 1>}
 Pose_32 {type=Curve, loc=<0 0 0>, rot=<0 0 0 1>}
 Pose_33 {type=Curve, loc=<0 0 0>, rot=<0 0 0 1>}
 Pose_34 {type=Curve, loc=<0 0 0>, rot=<0 0 0 1>}
 Pose_35 {type=Curve, loc=<0 0 0>, rot=<0 0 0 1>}
 Pose_36 {type=Curve, loc=<0 0 0>, rot=<0 0 0 1>}
 Pose_37 {type=Curve, loc=<0 0 0>, rot=<0 0 0 1>}
 Pose_38 {type=Curve, loc=<0 0 0>, rot=<0 0 0 1>}
 Pose_39 {type=Curve, loc=<0 0 0>, rot=<0 0 0 1>}
 Pose_40 {type=Curve, loc=<0 0 0>, rot=<0 0 0 1>}
 Pose_41 {type=Curve, loc=<0 0 0>, rot=<0 0 0 1>}
 Pose_42 {type=Curve, loc=<0 0 0>, rot=<0 0 0 1>}
 Pose_43 {type=Curve, loc=<0 0 0>, rot=<0 0 0 1>}
 Pose_44 {type=Curve, loc=<0 0 0>, rot=<0 0 0 1>}
 Pose_45 {type=Curve, loc=<0 0 0>, rot=<0 0 0 1>}
 Pose_46 {type=Curve, loc=<0 0 0>, rot=<0 0 0 1>}
 Pose_47 {type=Curve, loc=<0 0 0>, rot=<0 0 0 1>}
 Pose_48 {type=Curve, loc=<0 0 0>, rot=<0 0 0 1>}
 Pose_49 {type=Curve, loc=<0 0 0>, rot=<0 0 0 1>}
 Pose_50 {type=Curve, loc=<0 0 0>, rot=<0 0 0 1>}
 Pose_51 {type=Curve, loc=<0 0 0>, rot=<0 0 0 1>}
 Pose_52 {type=Curve, loc=<0 0 0>, rot=<0 0 0 1>}
 Pose_53 {type=Curve, loc=<0 0 0>, rot=<0 0 0 1>}
 Pose_54 {type=Curve, loc=<0 0 0>, rot=<0 0 0 1>}
 Pose_55 {type=Curve, loc=<0 0 0>, rot=<0 0 0 1>}
 Pose_56 {type=Curve, loc=<0 0 0>, rot=<0 0 0 1>}
 Pose_57 {type=Curve, loc=<0 0 0>, rot=<0 0 0 1>}
 Pose_58 {type=Curve, loc=<0 0 0>, rot=<0 0 0 1>}
 Pose_59 {type=Curve, loc=<0 0 0>, rot=<0 0 0 1>}
 Pose_60 {type=Curve, loc=<0 0 0>, rot=<0 0 0 1>}
 Pose_61 {type=Curve, loc=<0 0 0>, rot=<0 0 0 1>}
 Pose_62 {type=Curve, loc=<0 0 0>, rot=<0 0 0 1>}
 Pose_63 {type=Curve, loc=<0 0 0>, rot=<0 0 0 1>}
 Pose_64 {type=Curve, loc=<0 0 0>, rot=<0 0 0 1>}
 Pose_65 {type=Curve, loc=<0 0 0>, rot=<0 0 0 1>}
 Pose_66 {type=Curve, loc=<0 0 0>, rot=<0 0 0 1>}
 Pose_67 {type=Curve, loc=<0 0 0>, rot=<0 0 0 1>}
 Pose_68 {type=Curve, loc=<0 0 0>, rot=<0 0 0 1>}
 Pose_69 {type=Curve, loc=<0 0 0>, rot=<0 0 0 1>}
 eyeLookDownRight {type=Curve, loc=<0 0 0>, rot=<0 0 0 1>}
 eyeLookOutRight {type=Curve, loc=<0 0 0>, rot=<0 0 0 1>}
 eyeLookUpRight {type=Curve, loc=<0 0 0>, rot=<0 0 0 1>}
 eyeLookOutLeft {type=Curve, loc=<0 0 0>, rot=<0 0 0 1>}
 eyeBlinkLeft {type=Curve, loc=<0 0 0>, rot=<0 0 0 1>}
 eyeBlinkRight {type=Curve, loc=<0 0 0>, rot=<0 0 0 1>}
 eyeSquintLeft {type=Curve, loc=<0 0 0>, rot=<0 0 0 1>}
 eyeSquintRight {type=Curve, loc=<0 0 0>, rot=<0 0 0 1>}
 eyeWideLeft {type=Curve, loc=<0 0 0>, rot=<0 0 0 1>}
 eyeWideRight {type=Curve, loc=<0 0 0>, rot=<0 0 0 1>}
 jawForward {type=Curve, loc=<0 0 0>, rot=<0 0 0 1>}
 jawLeft {type=Curve, loc=<0 0 0>, rot=<0 0 0 1>}
 jawRight {type=Curve, loc=<0 0 0>, rot=<0 0 0 1>}
 jawOpen {type=Curve, loc=<0 0 0>, rot=<0 0 0 1>}
 mouthClose {type=Curve, loc=<0 0 0>, rot=<0 0 0 1>}
 mouthFunnel {type=Curve, loc=<0 0 0>, rot=<0 0 0 1>}
 mouthPucker {type=Curve, loc=<0 0 0>, rot=<0 0 0 1>}
 mouthLeft {type=Curve, loc=<0 0 0>, rot=<0 0 0 1>}
 mouthRight {type=Curve, loc=<0 0 0>, rot=<0 0 0 1>}
 mouthSmileLeft {type=Curve, loc=<0 0 0>, rot=<0 0 0 1>}
 mouthSmileRight {type=Curve, loc=<0 0 0>, rot=<0 0 0 1>}
 mouthFrownLeft {type=Curve, loc=<0 0 0>, rot=<0 0 0 1>}
 mouthFrownRight {type=Curve, loc=<0 0 0>, rot=<0 0 0 1>}
 mouthDimpleLeft {type=Curve, loc=<0 0 0>, rot=<0 0 0 1>}
 mouthDimpleRight {type=Curve, loc=<0 0 0>, rot=<0 0 0 1>}
 mouthStretchLeft {type=Curve, loc=<0 0 0>, rot=<0 0 0 1>}
 mouthStretchRight {type=Curve, loc=<0 0 0>, rot=<0 0 0 1>}
 mouthRollLower {type=Curve, loc=<0 0 0>, rot=<0 0 0 1>}
 mouthRollUpper {type=Curve, loc=<0 0 0>, rot=<0 0 0 1>}
 mouthShrugLower {type=Curve, loc=<0 0 0>, rot=<0 0 0 1>}
 mouthShrugUpper {type=Curve, loc=<0 0 0>, rot=<0 0 0 1>}
 mouthPressLeft {type=Curve, loc=<0 0 0>, rot=<0 0 0 1>}
 mouthPressRight {type=Curve, loc=<0 0 0>, rot=<0 0 0 1>}
 mouthLowerDownLeft {type=Curve, loc=<0 0 0>, rot=<0 0 0 1>}
 mouthLowerDownRight {type=Curve, loc=<0 0 0>, rot=<0 0 0 1>}
 mouthUpperUpLeft {type=Curve, loc=<0 0 0>, rot=<0 0 0 1>}
 mouthUpperUpRight {type=Curve, loc=<0 0 0>, rot=<0 0 0 1>}
 browDownLeft {type=Curve, loc=<0 0 0>, rot=<0 0 0 1>}
 browDownRight {type=Curve, loc=<0 0 0>, rot=<0 0 0 1>}
 browInnerUp {type=Curve, loc=<0 0 0>, rot=<0 0 0 1>}
 browOuterUpLeft {type=Curve, loc=<0 0 0>, rot=<0 0 0 1>}
 browOuterUpRight {type=Curve, loc=<0 0 0>, rot=<0 0 0 1>}
 cheekPuff {type=Curve, loc=<0 0 0>, rot=<0 0 0 1>}
 cheekSquintLeft {type=Curve, loc=<0 0 0>, rot=<0 0 0 1>}
 cheekSquintRight {type=Curve, loc=<0 0 0>, rot=<0 0 0 1>}
 noseSneerLeft {type=Curve, loc=<0 0 0>, rot=<0 0 0 1>}
 noseSneerRight {type=Curve, loc=<0 0 0>, rot=<0 0 0 1>}
 C_glabella_down {type=Curve, loc=<0 0 0>, rot=<0 0 0 1>}
 L_brow_squeeze_in {type=Curve, loc=<0 0 0>, rot=<0 0 0 1>}
 R_brow_squeeze_in {type=Curve, loc=<0 0 0>, rot=<0 0 0 1>}
 L_upper_lip_lower {type=Curve, loc=<0 0 0>, rot=<0 0 0 1>}
 R_upper_lip_lower {type=Curve, loc=<0 0 0>, rot=<0 0 0 1>}
 L_lower_lip_up {type=Curve, loc=<0 0 0>, rot=<0 0 0 1>}
 R_lower_lip_up {type=Curve, loc=<0 0 0>, rot=<0 0 0 1>}
 L_lip_corner_narrow {type=Curve, loc=<0 0 0>, rot=<0 0 0 1>}
 R_lip_corner_narrow {type=Curve, loc=<0 0 0>, rot=<0 0 0 1>}
 phoneme_ch {type=Curve, loc=<0 0 0>, rot=<0 0 0 1>}
 phoneme_mbp {type=Curve, loc=<0 0 0>, rot=<0 0 0 1>}
 phoneme_fv {type=Curve, loc=<0 0 0>, rot=<0 0 0 1>}
 mouth_up {type=Curve, loc=<0 0 0>, rot=<0 0 0 1>}
 mouth_down {type=Curve, loc=<0 0 0>, rot=<0 0 0 1>}
 L_cheek_up {type=Curve, loc=<0 0 0>, rot=<0 0 0 1>}
 R_cheek_up {type=Curve, loc=<0 0 0>, rot=<0 0 0 1>}
 L_upperLidDown {type=Curve, loc=<0 0 0>, rot=<0 0 0 1>}
 R_upperLidDown {type=Curve, loc=<0 0 0>, rot=<0 0 0 1>}
 jaw_up {type=Curve, loc=<0 0 0>, rot=<0 0 0 1>}
 jaw_back {type=Curve, loc=<0 0 0>, rot=<0 0 0 1>}
 tongue_forward {type=Curve, loc=<0 0 0>, rot=<0 0 0 1>}
 tongue_back {type=Curve, loc=<0 0 0>, rot=<0 0 0 1>}
 tongue_up {type=Curve, loc=<0 0 0>, rot=<0 0 0 1>}
 tongue_down {type=Curve, loc=<0 0 0>, rot=<0 0 0 1>}
 basePose {type=Curve, loc=<0 0 0>, rot=<0 0 0 1>}
 neutral {type=Curve, loc=<0 0 0>, rot=<0 0 0 1>}
 cheek_puff_low {type=Curve, loc=<0 0 0>, rot=<0 0 0 1>}
 Desert_Island_Jonesy_Mouth {type=Curve, loc=<0 0 0>, rot=<0 0 0 1>}
 HT_L1_TransitionCurve {type=Curve, loc=<0 0 0>, rot=<0 0 0 1>}
 Hair_CapShape {type=Curve, loc=<0 0 0>, rot=<0 0 0 1>}
 Hair_CapShape1 {type=Curve, loc=<0 0 0>, rot=<0 0 0 1>}
 jekyll {type=Curve, loc=<0 0 0>, rot=<0 0 0 1>}
 JH_TransformCurve {type=Curve, loc=<0 0 0>, rot=<0 0 0 1>}
 eyes_LOD0 {type=Curve, loc=<0 0 0>, rot=<0 0 0 1>}
 MaterialCurveA {type=Curve, loc=<0 0 0>, rot=<0 0 0 1>}
 MaterialCurveB {type=Curve, loc=<0 0 0>, rot=<0 0 0 1>}
 MaterialCurveC {type=Curve, loc=<0 0 0>, rot=<0 0 0 1>}
 MaterialCurveD {type=Curve, loc=<0 0 0>, rot=<0 0 0 1>}
 MaterialCurveE {type=Curve, loc=<0 0 0>, rot=<0 0 0 1>}
 Custom_Frontend_Face_Idle {type=Curve, loc=<0 0 0>, rot=<0 0 0 1>}
 blendParent1 {type=Curve, loc=<0 0 0>, rot=<0 0 0 1>}
 BaseHeadDefaultFaceIdle {type=Curve, loc=<0 0 0>, rot=<0 0 0 1>}
 VeiledTransitionGlowAmount {type=Curve, loc=<0 0 0>, rot=<0 0 0 1>}
 is_3L {type=Curve, loc=<0 0 0>, rot=<0 0 0 1>}
 Hair_Hat_LOD0 {type=Curve, loc=<0 0 0>, rot=<0 0 0 1>}
 Hair_Helmet_LOD0 {type=Curve, loc=<0 0 0>, rot=<0 0 0 1>}
 Hair_Null_LOD0 {type=Curve, loc=<0 0 0>, rot=<0 0 0 1>}
 mouth_close_pose {type=Curve, loc=<0 0 0>, rot=<0 0 0 1>}
 mouth_pucker_pose {type=Curve, loc=<0 0 0>, rot=<0 0 0 1>}
 mouth_left_pose {type=Curve, loc=<0 0 0>, rot=<0 0 0 1>}
 mouth_right_pose {type=Curve, loc=<0 0 0>, rot=<0 0 0 1>}
 L_mouth_stretch_pose {type=Curve, loc=<0 0 0>, rot=<0 0 0 1>}
 R_mouth_stretch_pose {type=Curve, loc=<0 0 0>, rot=<0 0 0 1>}
 mouth_roll_lower_pose {type=Curve, loc=<0 0 0>, rot=<0 0 0 1>}
 mouth_roll_upper_pose {type=Curve, loc=<0 0 0>, rot=<0 0 0 1>}
 mouth_shrug_lower_pose {type=Curve, loc=<0 0 0>, rot=<0 0 0 1>}
 mouth_shrug_upper_pose {type=Curve, loc=<0 0 0>, rot=<0 0 0 1>}
 L_mouth_press_pose {type=Curve, loc=<0 0 0>, rot=<0 0 0 1>}
 R_mouth_press_pose {type=Curve, loc=<0 0 0>, rot=<0 0 0 1>}
 cheek_puff_pose {type=Curve, loc=<0 0 0>, rot=<0 0 0 1>}
 L_nose_wrinkler_pose {type=Curve, loc=<0 0 0>, rot=<0 0 0 1>}
 R_nose_wrinkler_pose {type=Curve, loc=<0 0 0>, rot=<0 0 0 1>}
 base {type=Curve, loc=<0 0 0>, rot=<0 0 0 1>}
 eyeLookDownLeft {type=Curve, loc=<0 0 0>, rot=<0 0 0 1>}
 eyeLookInLeft {type=Curve, loc=<0 0 0>, rot=<0 0 0 1>}
 eyeLookUpLeft {type=Curve, loc=<0 0 0>, rot=<0 0 0 1>}
 eyeLookInRight {type=Curve, loc=<0 0 0>, rot=<0 0 0 1>}
 is_face_bone_anim {type=Curve, loc=<0 0 0>, rot=<0 0 0 1>}
 FullHourglass {type=Curve, loc=<0 0 0>, rot=<0 0 0 1>}
 LeanRectangle {type=Curve, loc=<0 0 0>, rot=<0 0 0 1>}
 InvertedTriangle {type=Curve, loc=<0 0 0>, rot=<0 0 0 1>}
 L_eye {type=Curve, loc=<0 0 0>, rot=<0 0 0 1>}
 R_eye {type=Curve, loc=<0 0 0>, rot=<0 0 0 1>}
 head {type=Curve, loc=<0 0 0>, rot=<0 0 0 1>}
 dyn_ear_l_master {type=Curve, loc=<0 0 0>, rot=<0 0 0 1>}
 root {type=Curve, loc=<0 0 0>, rot=<0 0 0 1>}
 dyn_ear02_l {type=Curve, loc=<0 0 0>, rot=<0 0 0 1>}
 dyn_ear03_l {type=Curve, loc=<0 0 0>, rot=<0 0 0 1>}
 dyn_ear04_l {type=Curve, loc=<0 0 0>, rot=<0 0 0 1>}
 dyn_ear05_l {type=Curve, loc=<0 0 0>, rot=<0 0 0 1>}
 dyn_ear01_l {type=Curve, loc=<0 0 0>, rot=<0 0 0 1>}
 flipbook_face_L_eye_index {type=Curve, loc=<0 0 0>, rot=<0 0 0 1>}
 flipbook_face_R_eye_index {type=Curve, loc=<0 0 0>, rot=<0 0 0 1>}
 flipbook_face_L_brow_index {type=Curve, loc=<0 0 0>, rot=<0 0 0 1>}
 flipbook_face_R_brow_index {type=Curve, loc=<0 0 0>, rot=<0 0 0 1>}
 flipbook_face_mouth_index {type=Curve, loc=<0 0 0>, rot=<0 0 0 1>}
 flipbook_face_bypass_mapping {type=Curve, loc=<0 0 0>, rot=<0 0 0 1>}
 flipbook_face_is_idling {type=Curve, loc=<0 0 0>, rot=<0 0 0 1>}
 head_SPR {type=Curve, loc=<0 0 0>, rot=<0 0 0 1>}
]
