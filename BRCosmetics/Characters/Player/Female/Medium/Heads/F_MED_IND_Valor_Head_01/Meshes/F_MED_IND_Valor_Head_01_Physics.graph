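# Physics asset for the F_MED_IND_Valor_Head_01 female cosmetic head, setting up capsule bodies for a dynamic hair strand and the upper torso.
# Three dyn_hair_05 bodies chain from a tiny head capsule through constraints with swing and twist limit angles.
# Larger capsules cover spine_04, spine_05 and both clavicles, with the clavicles joined to spine_05 by free-rotation constraints.
# No constraint links the torso bodies to the head and hair chain.
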PhysicsAsset /BRCosmetics/Characters/Player/Female/Medium/Heads/F_MED_IND_Valor_Head_01/Meshes/F_MED_IND_Valor_Head_01_Physics
bodies {
 clavicle_l [{shapeType=Capsule, length=13.3235, radius=7.4594, shapeName=clavicle_l_capsule, center=<5.9481 0 2.044>, rotation=<rot 90 0 0>}]
 clavicle_r [{shapeType=Capsule, length=13.3235, radius=7.4594, shapeName=clavicle_l_capsule, center=<-5.7152 0 -2.1381>, rotation=<rot 90 0 0>}]
 dyn_hair_05 [{shapeType=Capsule, length=14.9175, radius=4.3502, shapeName=dyn_hair_05_capsule, center=<0 -2.585 -5.2457>, rotation=<rot 90 0 0>}]
 dyn_hair_05_a [{shapeType=Capsule, length=14.0162, radius=3.9499, shapeName=dyn_hair_05_a_capsule, center=<0 0 -2.8905>, rotation=<rot 90 0 0>}]
 dyn_hair_05_b [{shapeType=Capsule, length=9.6056, radius=4.0416, shapeName=dyn_hair_05_b_capsule, center=<0 -1.4036 -3.2042>, rotation=<rot 90 0 0>}]
 head [{shapeType=Capsule, length=1, radius=1, shapeName=head_capsule}]
 spine_04 [{shapeType=Capsule, length=14.2138, radius=10.4418, shapeName=spine_04_capsule, center=<8.6254 0.6137 0>, rotation=<rot 0 -90 0>}]
 spine_05 [{shapeType=Capsule, length=16.9026, radius=10.0443, shapeName=spine_05_capsule, center=<5.1243 2.2969 0>}]
}
constraints [
 {bone1Name=dyn_hair_05_a, bone2Name=dyn_hair_05, swing1LimitDegrees=45, swing2LimitDegrees=45, twistLimitDegrees=40}
 {bone1Name=dyn_hair_05_b, bone2Name=dyn_hair_05_a, swing1LimitDegrees=45, swing2LimitDegrees=45, twistLimitDegrees=40}
 {bone1Name=dyn_hair_05, bone2Name=head, swing1LimitDegrees=45, swing2LimitDegrees=45, twistLimitDegrees=40}
 {
  bone1Name=clavicle_r
  bone2Name=spine_05
  swing1LimitDegrees=45
  swing1Motion=Free
  swing2LimitDegrees=45
  swing2Motion=Free
  twistLimitDegrees=45
  twistMotion=Free
 }
 {
  bone1Name=clavicle_l
  bone2Name=spine_05
  swing1LimitDegrees=45
  swing1Motion=Free
  swing2LimitDegrees=45
  swing2Motion=Free
  twistLimitDegrees=45
  twistMotion=Free
 }
]
