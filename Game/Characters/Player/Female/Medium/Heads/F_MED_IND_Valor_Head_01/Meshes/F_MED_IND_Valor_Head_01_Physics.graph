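# Physics asset for the female medium Valor head 01 player mesh, defining bodies for a dynamic hair chain plus the upper spine, neck and head.
# Capsules cover spine_04 up through head, linked by constraints with free swing and twist.
# Three dyn_hair_05 bones chain off the head, the tip using two capsules.
# Hair joints lock swing1 and limit swing2 and twist, with much tighter twist at the head attachment.
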PhysicsAsset /Game/Characters/Player/Female/Medium/Heads/F_MED_IND_Valor_Head_01/Meshes/F_MED_IND_Valor_Head_01_Physics
bodies {
 dyn_hair_05 [{shapeType=Capsule, length=13.0799, radius=6.0935, center=<-0.0824 -3.8948 -2.1887>, rotation=<rot 90 0 0>}]
 dyn_hair_05_a [{shapeType=Capsule, length=13.5907, radius=6.2696, center=<-0.0826 -0.3124 -0.3862>, rotation=<rot 90 0 0>}]
 dyn_hair_05_b [
  {shapeType=Capsule, length=7.584, radius=5.0886, center=<-0.3221 -0.1935 -0.005>, rotation=<rot 90 0 0>}
  {shapeType=Capsule, length=7.25, radius=3, center=<0.01 -1.0754 -6.2189>, rotation=<rot 90 0 0>}
 ]
 head [{shapeType=Capsule, length=0, radius=10, center=<4.3354 -1.5577 -0.0001>}]
 neck_01 [{shapeType=Capsule, length=1.3109, radius=5.0242, center=<1.4367 0.1963 0.063>, rotation=<rot 0 -90 0>}]
 neck_02 [{shapeType=Capsule, length=1.3122, radius=4.2815, center=<-0.4409 -0.2512 -0.0575>, rotation=<rot 0 -90 0>}]
 spine_04 [{shapeType=Capsule, length=0.505, radius=0.505, rotation=<rot 0 -90 0>}]
 spine_05 [{shapeType=Capsule, length=9.4942, radius=11.1907, center=<1.958 2.0356 0>, rotation=<rot 0 -90 0>}]
}
constraints [
 {
  bone1Name=head
  bone2Name=neck_02
  swing1LimitDegrees=45
  swing1Motion=Free
  swing2LimitDegrees=45
  swing2Motion=Free
  twistLimitDegrees=45
  twistMotion=Free
 }
 {
  bone1Name=spine_05
  bone2Name=spine_04
  swing1LimitDegrees=45
  swing1Motion=Free
  swing2LimitDegrees=45
  swing2Motion=Free
  twistLimitDegrees=45
  twistMotion=Free
 }
 {
  bone1Name=neck_01
  bone2Name=spine_05
  swing1LimitDegrees=45
  swing1Motion=Free
  swing2LimitDegrees=45
  swing2Motion=Free
  twistLimitDegrees=45
  twistMotion=Free
 }
 {
  bone1Name=neck_02
  bone2Name=neck_01
  swing1LimitDegrees=45
  swing1Motion=Free
  swing2LimitDegrees=45
  swing2Motion=Free
  twistLimitDegrees=45
  twistMotion=Free
 }
 {bone1Name=dyn_hair_05_a, bone2Name=dyn_hair_05, swing1LimitDegrees=45, swing1Motion=Locked, swing2LimitDegrees=20, twistLimitDegrees=40}
 {bone1Name=dyn_hair_05_b, bone2Name=dyn_hair_05_a, swing1LimitDegrees=45, swing1Motion=Locked, swing2LimitDegrees=20, twistLimitDegrees=40}
 {bone1Name=dyn_hair_05, bone2Name=head, swing1LimitDegrees=45, swing1Motion=Locked, swing2LimitDegrees=20, twistLimitDegrees=5}
]
