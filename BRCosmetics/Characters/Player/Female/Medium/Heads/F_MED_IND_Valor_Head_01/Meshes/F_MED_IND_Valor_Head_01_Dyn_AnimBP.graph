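# Post-process style AnimBlueprint that adds dynamic hair motion to the female Valor head cosmetic.
# A cached Control Rig pose feeds head-space RigidBody simulation, bypassed when backbling is detected, then low-alpha AnimDynamics and additive bone offsets driven by the head_SPR curve and skydiving.
# Skydiving also alters RigidBody gravity and alpha.
# Playlists using custom character parts instead shrink the head bone; several unconnected dynamics chains remain.
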AnimBlueprint /BRCosmetics/Characters/Player/Female/Medium/Heads/F_MED_IND_Valor_Head_01/Meshes/F_MED_IND_Valor_Head_01_Dyn_AnimBP
target_skeleton=/Game/Characters/Player/Common/Fortnite_Base_Head/Fortnite_Base_Head_Export_Skeleton
graphs [
 AnimGraph {nodes [
   n1 = Anim Dynamics - Bone: dyn_hair_05(Alpha=0.5, ExternalForce="0,0,0") {
    angularDampingOverride=0.999
    bGravityOverrideInSimSpace=true
    bOverrideAngularDamping=true
    bUseGravityOverride=true
    simulationSpace=RootRelative
    alphaBoolBlend {blendOption=Linear}
    boundBone {boneName=dyn_hair_05}
    gravityOverride {z=-980}
    physicsBodyDefinitions [{
      collisionType=CustomSphere
      sphereCollisionRadius=3
      boundBone {boneName=dyn_hair_05}
      boxExtents=<30 30 4>
      constraintSetup {angularLimitsMax=<0 15 0>, angularLimitsMin=<0 -15 0>, angularTarget=<0 0 0>}
      localJointOffset=<0 0 -4>
     }]
    planarLimits [{drivingBone {boneName=neck_02}, planeTransform {location=<10 0 0>, rotation=<rot 0 30.0001 -89.9999>, scale=<0.5 0.5 0.5>}}]
   }
   n2 = Anim Dynamics - Bone: dyn_hair_05_a(ComponentPose=@n1, Alpha=0.7, ExternalForce="0,0,0") {
    angularDampingOverride=0.999
    bGravityOverrideInSimSpace=true
    bOverrideAngularDamping=true
    bUseGravityOverride=true
    gravityScale=2
    simulationSpace=RootRelative
    alphaBoolBlend {blendOption=Linear}
    boundBone {boneName=dyn_hair_05_a}
    gravityOverride {z=-980}
    physicsBodyDefinitions [{
      collisionType=CustomSphere
      sphereCollisionRadius=6
      boundBone {boneName=dyn_hair_05_a}
      boxExtents=<30 30 4>
      constraintSetup {angularTargetAxis=AxisZ, twistAxis=AxisZ, angularLimitsMax=<15 15 0>, angularLimitsMin=<-15 -15 0>, angularTarget=<0 0 0>}
      localJointOffset=<0 0 -10>
     }]
    planarLimits [{drivingBone {boneName=spine_05}, planeTransform {location=<10 -6 0>, rotation=<rot 0 30 -89.9999>, scale=<0.5 0.5 0.5>}}]
   }
   Anim Dynamics - Bone: dyn_hair_05_b(ComponentPose=@n2, Alpha=0.7, ExternalForce="0,0,0") {
    angularDampingOverride=0.999
    bGravityOverrideInSimSpace=true
    bOverrideAngularDamping=true
    bUseGravityOverride=true
    simulationSpace=RootRelative
    alphaBoolBlend {blendOption=Linear}
    boundBone {boneName=dyn_hair_05_b}
    gravityOverride {z=-980}
    physicsBodyDefinitions [{
      collisionType=CustomSphere
      sphereCollisionRadius=5
      boundBone {boneName=dyn_hair_05_b}
      boxExtents=<30 30 10>
      constraintSetup {angularTargetAxis=AxisZ, twistAxis=AxisZ, angularLimitsMax=<15 15 0>, angularLimitsMin=<-15 -15 0>, angularTarget=<0 0 0>}
      localJointOffset=<0 0 -10>
     }]
   }
   n3 = Anim Dynamics - Bone: dyn_hair_07(Alpha=1, ExternalForce="0,0,0") {
    alpha=0.9
    angularSpringConstant=1500
    bAngularSpring=true
    bGravityOverrideInSimSpace=true
    gravityScale=0.75
    numSolverIterationsPostUpdate=4
    numSolverIterationsPreUpdate=16
    alphaBoolBlend {blendOption=Linear}
    boundBone {boneName=dyn_hair_07}
    physicsBodyDefinitions [{
      sphereCollisionRadius=0
      boundBone {boneName=dyn_hair_07}
      boxExtents=<15 5 5>
      constraintSetup {angularTargetAxis=AxisZ, twistAxis=AxisY, angularLimitsMax=<2 0 15>, angularLimitsMin=<-2 0 -15>, angularTarget=<0 0 1>}
      localJointOffset=<0 -1 0>
     }]
   }
   Anim Dynamics - Bone: dyn_hair_07_a(ComponentPose=@n3, Alpha=1, ExternalForce="0,0,0") {
    alpha=0.9
    angularSpringConstant=1500
    bAngularSpring=true
    bGravityOverrideInSimSpace=true
    gravityScale=0.75
    numSolverIterationsPostUpdate=4
    numSolverIterationsPreUpdate=16
    alphaBoolBlend {blendOption=Linear}
    boundBone {boneName=dyn_hair_07_a}
    physicsBodyDefinitions [{
      sphereCollisionRadius=0
      boundBone {boneName=dyn_hair_07_a}
      boxExtents=<15 5 5>
      constraintSetup {angularTargetAxis=AxisZ, twistAxis=AxisY, angularLimitsMax=<2 0 15>, angularLimitsMin=<-2 0 -15>, angularTarget=<0 0 1>}
      localJointOffset=<0 -1 0>
     }]
   }
   n4 = Anim Dynamics - Bone: dyn_hair_06_a(Alpha=1, ExternalForce="0,0,0") {
    bGravityOverrideInSimSpace=true
    simulationSpace=Actor
    alphaBoolBlend {blendOption=Linear}
    boundBone {boneName=dyn_hair_06_a}
    physicsBodyDefinitions [{
      sphereCollisionRadius=0
      boundBone {boneName=dyn_hair_06_a}
      boxExtents=<30 30 1>
      constraintSetup {angularLimitsMax=<0 30 30>, angularLimitsMin=<0 -30 -30>, angularTarget=<0 0 0>}
      localJointOffset=<0 0 -1>
     }]
   }
   n5 = Anim Dynamics - Bone: dyn_hair_06_b(ComponentPose=@n4, Alpha=1, ExternalForce="0,0,0") {
    bGravityOverrideInSimSpace=true
    simulationSpace=Actor
    alphaBoolBlend {blendOption=Linear}
    boundBone {boneName=dyn_hair_06_b}
    physicsBodyDefinitions [{
      sphereCollisionRadius=0
      boundBone {boneName=dyn_hair_06_b}
      boxExtents=<30 30 1>
      constraintSetup {angularLimitsMax=<30 0 0>, angularLimitsMin=<-30 0 0>, angularTarget=<0 0 0>}
      localJointOffset=<0 0 -1>
     }]
   }
   n6 = Anim Dynamics - Bone: dyn_hair_07(ComponentPose=@n5, Alpha=1, ExternalForce="0,0,0") {
    angularSpringConstant=50
    bAngularSpring=true
    bGravityOverrideInSimSpace=true
    bLinearSpring=true
    linearSpringConstant=50
    simulationSpace=Actor
    alphaBoolBlend {blendOption=Linear}
    boundBone {boneName=dyn_hair_07}
    physicsBodyDefinitions [{
      sphereCollisionRadius=0
      boundBone {boneName=dyn_hair_07}
      boxExtents=<30 30 1>
      constraintSetup {angularLimitsMax=<30 0 30>, angularLimitsMin=<-30 0 -30>, angularTarget=<0 0 0>}
      localJointOffset=<0 0 0.5>
     }]
   }
   n7 = Anim Dynamics - Bone: dyn_hair_08(ComponentPose=@n6, Alpha=1, ExternalForce="0,0,0") {
    angularSpringConstant=50
    bAngularSpring=true
    bGravityOverrideInSimSpace=true
    bLinearSpring=true
    linearSpringConstant=50
    simulationSpace=Actor
    alphaBoolBlend {blendOption=Linear}
    boundBone {boneName=dyn_hair_08}
    physicsBodyDefinitions [{
      sphereCollisionRadius=0
      boundBone {boneName=dyn_hair_08}
      boxExtents=<30 30 1>
      constraintSetup {angularLimitsMax=<30 0 30>, angularLimitsMin=<-30 0 -30>, angularTarget=<0 0 0>}
      localJointOffset=<0 0 0.5>
     }]
   }
   n8 = Anim Dynamics - Bone: dyn_hair_08_a(ComponentPose=@n7, Alpha=1, ExternalForce="0,0,0") {
    angularSpringConstant=50
    bAngularSpring=true
    bGravityOverrideInSimSpace=true
    bLinearSpring=true
    linearSpringConstant=50
    simulationSpace=Actor
    alphaBoolBlend {blendOption=Linear}
    boundBone {boneName=dyn_hair_08_a}
    physicsBodyDefinitions [{
      sphereCollisionRadius=0
      boundBone {boneName=dyn_hair_08_a}
      boxExtents=<30 30 1>
      constraintSetup {angularLimitsMax=<30 0 30>, angularLimitsMin=<-30 0 -30>, angularTarget=<0 0 0>}
      localJointOffset=<0 0 0.5>
     }]
   }
   Anim Dynamics - Bone: dyn_hair_07_a(ComponentPose=@n8, Alpha=1, ExternalForce="0,0,0") {
    angularSpringConstant=50
    bAngularSpring=true
    bGravityOverrideInSimSpace=true
    bLinearSpring=true
    linearSpringConstant=50
    simulationSpace=Actor
    alphaBoolBlend {blendOption=Linear}
    boundBone {boneName=dyn_hair_07_a}
    physicsBodyDefinitions [{
      sphereCollisionRadius=0
      boundBone {boneName=dyn_hair_07_a}
      boxExtents=<30 30 1>
      constraintSetup {angularLimitsMax=<30 0 30>, angularLimitsMin=<-30 0 -30>, angularTarget=<0 0 0>}
      localJointOffset=<0 0 0.5>
     }]
   }
   n9 = Anim Dynamics - Bone: dyn_hair_08(Alpha=1, ExternalForce="0,0,0") {
    alpha=0.9
    angularSpringConstant=1500
    bGravityOverrideInSimSpace=true
    gravityScale=0.75
    numSolverIterationsPostUpdate=4
    numSolverIterationsPreUpdate=16
    alphaBoolBlend {blendOption=Linear}
    boundBone {boneName=dyn_hair_08}
    physicsBodyDefinitions [{
      sphereCollisionRadius=0
      boundBone {boneName=dyn_hair_08}
      boxExtents=<15 5 5>
      constraintSetup {angularTargetAxis=AxisZ, twistAxis=AxisY, angularLimitsMax=<2 0 5>, angularLimitsMin=<-2 0 -5>, angularTarget=<0 0 1>}
      localJointOffset=<0 -1 0>
     }]
   }
   Anim Dynamics - Bone: dyn_hair_08_a(ComponentPose=@n9, Alpha=1, ExternalForce="0,0,0") {
    alpha=0.9
    angularSpringConstant=1500
    bAngularSpring=true
    bGravityOverrideInSimSpace=true
    gravityScale=0.75
    numSolverIterationsPostUpdate=4
    numSolverIterationsPreUpdate=16
    alphaBoolBlend {blendOption=Linear}
    boundBone {boneName=dyn_hair_08_a}
    physicsBodyDefinitions [{
      sphereCollisionRadius=0
      boundBone {boneName=dyn_hair_08_a}
      boxExtents=<15 5 5>
      constraintSetup {angularTargetAxis=AxisZ, twistAxis=AxisY, angularLimitsMax=<2 0 15>, angularLimitsMin=<-2 0 -15>, angularTarget=<0 0 1>}
      localJointOffset=<0 -1 0>
     }]
   }
   RigidBody() {
    overridePhysicsAsset=/Game/Characters/Player/Female/Medium/Heads/F_MED_IND_Valor_Head_01/Meshes/F_MED_IND_Valor_Head_01_Physics
    alphaBoolBlend {blendOption=Linear}
    componentAppliedLinearAccClamp=<0 0 0>
    componentLinearAccScale=<1.2 1.2 1.2>
   }
   RigidBody() {
    overridePhysicsAsset=&F_MED_IND_Valor_Head_01_Physics
    alphaBoolBlend {blendOption=Linear}
   }
   n10 = "Transform (Modify) Bone - Bone: head"!(ComponentPose=Local To Component(LocalPose=Use cached pose 'PreScale Cache'()), Alpha=1) -> (Pose) {scaleMode=BMM_Replace, alphaBoolBlend {blendOption=Linear}, boneToModify {boneName=head}, scale=<0.01 0.01 0.01>}
   n11 = RigidBody!(
    ComponentPose=Local To Component(LocalPose=Use cached pose 'PreScale Cache'())
    Alpha=Get RBAN_Alpha() -> (RBAN_Alpha:double)
    OverrideWorldGravity:Vector=Get Override World Gravity() -> (Override World Gravity:Vector)
    ExternalForce
   ) -> (Pose) {
    bOverrideWorldGravity=true
    overridePhysicsAsset=&/F_MED_IND_Valor_Head_01_Physics
    simulationSpace=BaseBoneSpace
    alphaBoolBlend {blendOption=Linear}
    baseBoneRef {boneName=head}
    overrideWorldGravity {y=400, z=-980}
    simSpaceSettings {
     dampingAlpha=0.4
     maxAngularAcceleration=4000
     maxAngularVelocity=8
     maxLinearAcceleration=3000
     maxLinearVelocity=600
     velocityScaleZ=0.5
     worldAlpha=1
     externalLinearDragV=<0.2 0.2 0.2>
    }
   }
   n12 = Component To Local(ComponentPose=@n11)
   n13 = Use cached pose 'PreScale Cache'()
   n14 = Get isWearingPack() -> (isWearingPack:bool)
   n15 = Anim Dynamics - Bone: dyn_hair_07(
    ComponentPose=Local To Component(LocalPose=Blend Poses by bool(BlendPose_0=@n13, BlendPose_1=@n12, bActiveValue=@n14) {blendType=Linear})
    Alpha=0.15
    ExternalForce="0,0,0"
   ) {
    angularBiasOverride=1
    angularDampingOverride=0.999
    bGravityOverrideInSimSpace=true
    bOverrideAngularDamping=true
    bUseGravityOverride=true
    lODThreshold=1
    numSolverIterationsPostUpdate=4
    numSolverIterationsPreUpdate=16
    simulationSpace=RootRelative
    alphaBoolBlend {blendOption=Linear}
    boundBone {boneName=dyn_hair_07}
    gravityOverride {z=-980}
    physicsBodyDefinitions [{
      sphereCollisionRadius=0
      boundBone {boneName=dyn_hair_07}
      boxExtents=<5 2 10>
      constraintSetup {angularTargetAxis=AxisZ, twistAxis=AxisZ, angularLimitsMax=<25 45 0>, angularLimitsMin=<0 -45 0>, angularTarget=<0 0 0>}
      localJointOffset=<0 1 -1>
     }]
   }
   n16 = Anim Dynamics - Bone: dyn_hair_07_a(ComponentPose=@n15, Alpha=0.15, ExternalForce="0,0,0") {
    angularBiasOverride=1
    angularDampingOverride=0.999
    bGravityOverrideInSimSpace=true
    bOverrideAngularDamping=true
    bUseGravityOverride=true
    lODThreshold=1
    numSolverIterationsPostUpdate=4
    numSolverIterationsPreUpdate=16
    simulationSpace=RootRelative
    alphaBoolBlend {blendOption=Linear}
    boundBone {boneName=dyn_hair_07_a}
    gravityOverride {z=-980}
    physicsBodyDefinitions [{
      sphereCollisionRadius=0
      boundBone {boneName=dyn_hair_07_a}
      boxExtents=<5 2 10>
      constraintSetup {angularTargetAxis=AxisZ, twistAxis=AxisZ, angularLimitsMax=<25 45 0>, angularLimitsMin=<0 -45 0>, angularTarget=<0 0 0>}
      localJointOffset=<0 1 -1>
     }]
   }
   n17 = Anim Dynamics - Bone: dyn_hair_08(ComponentPose=@n16, Alpha=0.15, ExternalForce="0,0,0") {
    angularBiasOverride=1
    angularDampingOverride=0.999
    bGravityOverrideInSimSpace=true
    bOverrideAngularDamping=true
    bUseGravityOverride=true
    lODThreshold=1
    numSolverIterationsPostUpdate=4
    numSolverIterationsPreUpdate=16
    simulationSpace=RootRelative
    alphaBoolBlend {blendOption=Linear}
    boundBone {boneName=dyn_hair_08}
    gravityOverride {z=-980}
    physicsBodyDefinitions [{
      sphereCollisionRadius=0
      boundBone {boneName=dyn_hair_08}
      boxExtents=<5.5 1.5 11>
      constraintSetup {angularTargetAxis=AxisZ, twistAxis=AxisZ, angularLimitsMax=<25 45 0>, angularLimitsMin=<0 -45 0>, angularTarget=<0 0 0>}
      localJointOffset=<0 1.5 -1>
     }]
   }
   n18 = Anim Dynamics - Bone: dyn_hair_08_a(ComponentPose=@n17, Alpha=0.15, ExternalForce="0,0,0") {
    angularBiasOverride=1
    angularDampingOverride=0.999
    bGravityOverrideInSimSpace=true
    bOverrideAngularDamping=true
    bUseGravityOverride=true
    lODThreshold=1
    numSolverIterationsPostUpdate=4
    numSolverIterationsPreUpdate=16
    simulationSpace=RootRelative
    alphaBoolBlend {blendOption=Linear}
    boundBone {boneName=dyn_hair_08_a}
    gravityOverride {z=-980}
    physicsBodyDefinitions [{
      sphereCollisionRadius=0
      boundBone {boneName=dyn_hair_08_a}
      boxExtents=<5.5 1.5 11>
      constraintSetup {angularTargetAxis=AxisZ, twistAxis=AxisZ, angularLimitsMax=<25 45 0>, angularLimitsMin=<0 -45 0>, angularTarget=<0 0 0>}
      localJointOffset=<0 1.5 -1>
     }]
   }
   n19 = Anim Dynamics - Bone: dyn_hair_06_a(ComponentPose=@n18, Alpha=0.2, ExternalForce="0,0,0") {
    angularBiasOverride=1
    angularDampingOverride=0.95
    angularSpringConstant=500
    bGravityOverrideInSimSpace=true
    bOverrideAngularDamping=true
    bUseGravityOverride=true
    lODThreshold=1
    numSolverIterationsPostUpdate=4
    numSolverIterationsPreUpdate=16
    simulationSpace=RootRelative
    alphaBoolBlend {blendOption=Linear}
    boundBone {boneName=dyn_hair_06_a}
    gravityOverride {z=-980}
    physicsBodyDefinitions [{
      sphereCollisionRadius=0
      boundBone {boneName=dyn_hair_06_a}
      boxExtents=<5 2 10>
      constraintSetup {angularTargetAxis=AxisZ, twistAxis=AxisZ, angularLimitsMax=<25 40 0>, angularLimitsMin=<0 -40 0>, angularTarget=<0 0 0>}
      localJointOffset=<0 0.5 -0.5>
     }]
   }
   n20 = Anim Dynamics - Bone: dyn_hair_06_b(ComponentPose=@n19, Alpha=0.25, ExternalForce="0,0,0") {
    angularBiasOverride=1
    angularDampingOverride=0.95
    angularSpringConstant=500
    bGravityOverrideInSimSpace=true
    bOverrideAngularDamping=true
    bUseGravityOverride=true
    lODThreshold=1
    numSolverIterationsPostUpdate=4
    numSolverIterationsPreUpdate=16
    simulationSpace=RootRelative
    alphaBoolBlend {blendOption=Linear}
    boundBone {boneName=dyn_hair_06_b}
    gravityOverride {z=-980}
    physicsBodyDefinitions [{
      sphereCollisionRadius=0
      boundBone {boneName=dyn_hair_06_b}
      boxExtents=<5 2 10>
      constraintSetup {angularTargetAxis=AxisZ, twistAxis=AxisZ, angularLimitsMax=<25 40 0>, angularLimitsMin=<0 -40 0>, angularTarget=<0 0 0>}
      localJointOffset=<0 0.5 -0.5>
     }]
   }
   n21 = Get headSPR_Alpha() -> (headSPR_Alpha:double)
   n22 = "Transform (Modify) Bone - Bone: dyn_hair_05_a"(ComponentPose=@n20, Alpha=@n21, Translation="0.918,2.060,1.013", Rotation="0.161,-0.079,-9.436", Scale="1,1,1") {
    rotationMode=BMM_Additive
    rotationSpace=BCS_BoneSpace
    translationMode=BMM_Additive
    translationSpace=BCS_BoneSpace
    alphaBoolBlend {blendOption=Linear}
    boneToModify {boneName=dyn_hair_05_a}
    rotation=<rot 0.161 -0.079 -9.436>
    translation=<0.918 2.06 1.013>
   }
   n23 = "Transform (Modify) Bone - Bone: dyn_hair_08"(ComponentPose=@n22, Alpha=@n21, Translation="-0.714,-0.587,3.384", Rotation="3.237,11.371,23.642", Scale="1,1,1") {
    rotationMode=BMM_Additive
    rotationSpace=BCS_BoneSpace
    translationMode=BMM_Additive
    translationSpace=BCS_BoneSpace
    alphaBoolBlend {blendOption=Linear}
    boneToModify {boneName=dyn_hair_08}
    rotation=<rot 3.237 11.371 23.642>
    translation=<-0.714 -0.587 3.384>
   }
   n24 = "Transform (Modify) Bone - Bone: dyn_hair_01"(ComponentPose=@n23, Alpha=@n21, Translation="0,0,0", Rotation="-6.979,1.001,20.743", Scale="1,1,1") {
    rotationMode=BMM_Additive
    rotationSpace=BCS_BoneSpace
    alphaBoolBlend {blendOption=Linear}
    boneToModify {boneName=dyn_hair_01}
    rotation=<rot -6.979 1.001 20.7429>
   }
   n25 = "Transform (Modify) Bone - Bone: dyn_hair_06"(ComponentPose=@n24, Alpha=@n21, Translation="0,0,0", Rotation="-0.043,0.000,20.628", Scale="1,1,1") {
    rotationMode=BMM_Additive
    rotationSpace=BCS_BoneSpace
    alphaBoolBlend {blendOption=Linear}
    boneToModify {boneName=dyn_hair_06}
    rotation=<rot -0.043 0 20.628>
   }
   n26 = Get bIsSkyDiving()
   n27 = "Transform (Modify) Bone - Bone: dyn_hair_05"!(ComponentPose=@n25, bAlphaBoolEnabled:bool=@n26, Translation="0,-1,0", Rotation="0,0,15", Scale="1,1,1") -> (Pose) {
    alphaInputType=Bool
    rotationMode=BMM_Additive
    rotationSpace=BCS_BoneSpace
    translationMode=BMM_Additive
    translationSpace=BCS_BoneSpace
    alphaBoolBlend {blendOption=Linear}
    boneToModify {boneName=dyn_hair_05}
    rotation {roll=15}
    translation {y=-1}
   }
   n28 = "Transform (Modify) Bone - Bone: dyn_hair_06"!(ComponentPose=@n27, bAlphaBoolEnabled:bool=@n26, Translation="0,0,0", Rotation="0,0,30", Scale="1,1,1") -> (Pose) {
    alphaInputType=Bool
    rotationMode=BMM_Additive
    rotationSpace=BCS_BoneSpace
    alphaBoolBlend {blendOption=Linear}
    boneToModify {boneName=dyn_hair_06}
    rotation {roll=30}
   }
   Output Pose(Result=Blend Poses by bool(
     BlendPose_0=Component To Local(ComponentPose=@n10)
     BlendPose_1=Component To Local(ComponentPose=@n28)
     BlendTime_0=0
     BlendTime_1=0
     bActiveValue=Get GetPlayerListUsesCustomCharacterParts() -> (GetPlayerListUsesCustomCharacterParts:bool)
    ) {blendType=Linear}) {layerGroup=None}
   AnimGraphNode_SubInput_0 = Input Pose()
   n29 = Control Rig(Source=@AnimGraphNode_SubInput_0) {
    alphaBoolBlend {blendOption=Linear}
    controlRigAssetReference {blueprintRigClass=&/F_MED_IND_Valor_Head_01_CtrlRig.F_MED_IND_Valor_Head_01_CtrlRig_C}
   }
   Save cached pose 'PreScale Cache'(Pose=@n29) {cacheName=PreScale Cache}
  ]}
 EventGraph {nodes [
   Event BlueprintUpdateAnimation()
   > Set headSPR_Alpha(headSPR_Alpha:double=GetCurveValue(CurveName=head_SPR)) -> (Output_Get:double)
   > n1 = Branch(Condition=Get bIsSkyDiving())
   > Set Override World Gravity(Override World Gravity:Vector="0,0,200") -> (Output_Get:Vector)
   > Set RBAN_Alpha(RBAN_Alpha:double=0.85) -> (Output_Get:double)
   Set Override World Gravity(execute=@n1.else, Override World Gravity="0,0,-980")
   > Set RBAN_Alpha(RBAN_Alpha=1)
   n2 = Event OnCharacterCustomizationLoaded()
   n3 = CheckForBackbling()
   > Set isWearingPack(isWearingPack:bool=OR Boolean!(A=@n3, B=@n3.bIsBackblingACape, C:bool=@n3.bIsBackblingLong) -> (ReturnValue)) -> (Output_Get:bool)
   > Set isWearingCape(isWearingCape=@n3.bIsBackblingLong)
   > Set CharacterMesh(CharacterMesh:Skeletal Mesh Component Object Reference=@n2.CustomBaseMesh) -> (Output_Get:Skeletal Mesh Component Object Reference)
   Event BlueprintInitializeAnimation()
   > Set GetPlayerListUsesCustomCharacterParts(GetPlayerListUsesCustomCharacterParts:bool=GetPlaylistUsesCustomCharacterParts!() -> (ReturnValue:bool)) -> (Output_Get:bool)
   TryGetPawnOwner()
  ]}
]
variables [
 CharacterMesh {type=SkeletalMeshComponent}
 GetPlayerListUsesCustomCharacterParts {type=boolean}
 headSPR_Alpha {type=number}
 isWearingCape {type=boolean}
 isWearingPack {type=boolean}
 Override World Gravity {type=Vector, default=<0 400 -980>}
 RBAN_Alpha {type=number, default=1}
]
